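# Labeled option selector widget for the frontend team select UI, built from a SettingsRotator between left and right arrow buttons.
# Initialize sets label, options and hover text; rotation stores the index and broadcasts Selection Changed, while Update Row Selector sets it externally.
# Hover writes to a supplied tooltip text block, cleared on leave; receiving focus centers the rotator.
# Button sizes adapt for mobile, and the right arrow mirrors the left's material.
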WidgetBlueprint /FortniteCoreUI/Frontend/TeamSelect/RotatorSelectorTeams1
graphs [
 Center On Widget {nodes [
   Center On Widget() {is=title K2Node_FunctionEntry}
   > CenterWidget!(execute, ToCenter:Widget Object Reference=Get SettingsRotator() -> (SettingsRotator:Settings Rotator Object Reference)) -> (then)
  ]}
 EventGraph {nodes [
   n1 = "On Rotated (SettingsRotator)"!() -> (OutputDelegate, then, Value:int=0) {is=title K2Node_ComponentBoundEvent}
   > n2 = Set SelectedIndex(SelectedIndex:int=@n1.Value) -> (Output_Get:int=0)
   > Call Selection Changed!(execute:Exec, self:Rotator Selector Teams 1 Object Reference, Selected Index:int=@n2) -> (then:Exec) {is=title K2Node_CallDelegate}
   "On Button Clicked (ButtonLeft)"!() -> (OutputDelegate, then, Button:Common Button Legacy Object Reference) {is=title K2Node_ComponentBoundEvent}
   > ShiftTextLeft(self:Common Rotator Object Reference=Get SettingsRotator())
   "On Button Clicked (ButtonRight)"!() -> (OutputDelegate, then, Button:Common Button Legacy Object Reference) {is=title K2Node_ComponentBoundEvent}
   > ShiftTextRight(self:Common Rotator Object Reference=Get SettingsRotator())
   Event Construct()
   > Update Size()
   Event PreConstruct()
   > Show Param Name!(execute, self, Show:bool=NOT Boolean(A=Get HideNameParam() -> (HideNameParam:bool))) -> (then)
   Event OnMouseLeave!() -> (OutputDelegate, then, MouseEvent:"Pointer Event Structure (by ref)") {is=title K2Node_Event}
   > "SetText (Text)"(self=Get Tab Tooltip Text() -> (Tab Tooltip Text:Common Text Object Reference))
   Event OnMouseEnter()
   > "SetText (Text)"(self=Get Tab Tooltip Text(), InText=Get HoverText() -> (HoverText:Text))
  ]}
 Initialize {nodes [
   n1 = Initialize!() -> (then, Display Text:Text, Options:Array of Texts, Hover Text:Text, Tab Tooltip:Common Text Object Reference) {is=title K2Node_FunctionEntry}
   > "SetText (Text)"(self=Get DisplayText() -> (DisplayText:Common Text Object Reference), InText=@n1)
   > PopulateTextLabels!(execute, self:Common Rotator Object Reference=Get SettingsRotator(), Labels:Array of Texts=@n1.Options) -> (then)
   > Set HoverText(HoverText:Text=@n1.Hover Text) -> (Output_Get:Text)
   > Set Tab Tooltip Text(Tab Tooltip Text:Common Text Object Reference=@n1.Tab Tooltip) -> (Output_Get:Common Text Object Reference)
  ]}
 OnFocusReceived {nodes [
   OnFocusReceived!() -> (then, MyGeometry:Geometry Structure, InFocusEvent:Focus Event Structure) {is=title K2Node_FunctionEntry}
   > Center On Widget() {is=title K2Node_CallFunction}
   > Return Node!(execute, ReturnValue:Event Reply Structure=Unhandled())
  ]}
 Selection Changed {nodes [Selection Changed!() -> (then, Selected Index:int) {is=title K2Node_FunctionEntry}]}
 Show Param Name {nodes [
   n1 = Show Param Name!() -> (then, Show:bool=~) {is=title K2Node_FunctionEntry}
   > SetVisibility(self=Get DisplayText(), InVisibility=Select(Option 0=Collapsed, Option 1=SelfHitTestInvisible, Index=@n1))
  ]}
 Update Options {nodes [
   n1 = Update Options!() -> (then, NewParam:Array of Texts) {is=title K2Node_FunctionEntry}
   > PopulateTextLabels(self=Get SettingsRotator(), Labels=@n1)
  ]}
 Update Row Selector {nodes [
   n1 = Update Row Selector!() -> (then, Currently Selected:int) {is=title K2Node_FunctionEntry}
   > n2 = Set SelectedIndex(SelectedIndex=@n1) -> (Output_Get=0)
   > SetSelectedItem!(execute, self:Common Rotator Object Reference=Get SettingsRotator(), InValue:int=@n2) -> (then)
  ]}
 Update Size {nodes [
   Update Size() {is=title K2Node_FunctionEntry}
   > n1 = Sequence!(execute) -> (then_0, then_1, then_2:Exec)
   n2 = Set members in Slate Brush!(
    execute:Exec=@n1.then_1
    StructRef:"Slate Brush Structure (by ref)"=Get IconBrush!(self:Icon Text Button Object Reference=Get ButtonLeft() -> (ButtonLeft:Icon Text Button Object Reference)) -> (IconBrush:Slate Brush Structure)
    ImageSize:Vector 2D Structure=Select(
     Option 0:Vector 2D Structure=Get SmallSize() -> (SmallSize:Vector 2D Structure)
     Option 1:Vector 2D Structure=Get LargeSize() -> (LargeSize:Vector 2D Structure)
     Index=IsMobileGame()
    ) -> (ReturnValue:Vector 2D Structure)
   ) -> (then:Exec, StructOut:"Slate Brush Structure (by ref)") {is=title K2Node_SetFieldsInStruct}
   > Set Icon!(execute, self:Icon Text Button Object Reference=Get ButtonLeft(), IconBrush:Slate Brush Structure=@n2) -> (then) {is=title K2Node_CallFunction}
   n3 = Set members in Slate Brush(
    execute=@n1.then_2
    StructRef=Get IconBrush(self=Get ButtonRight() -> (ButtonRight:Icon Text Button Object Reference))
    ImageSize=Select(Option 0=Get SmallSize(), Option 1=Get LargeSize(), Index=IsMobileGame())
   )
   > Set Icon(self=Get ButtonRight(), IconBrush=@n3)
   SetMinDimensions!(
    execute=@n1.then_0
    self:Common Button Base Object Reference=[Get ButtonLeft(), Get ButtonRight()]
    InMinWidth:int=96
    InMinHeight:int=Select(Option 0:int=48, Option 1:int=72, Index=IsMobileGame()) -> (ReturnValue:int)
   ) -> (then)
  ]}
]
variables [
 HideNameParam {type=boolean}
 HoverText {type=string, default=""}
 SelectedIndex {type=integer}
 Selection Changed {}
 Tab Tooltip Text {type=CommonTextBlock, default=null}
]
widgets {
 info {inheritedWidgetCount=0, parentClass=&CommonUserWidget, rootWidgetClass=&VerticalBox, widgetCount=8}
 widgets [
  VerticalBox_0 {widgetClassPath=&VerticalBox, widget=VerticalBox_0}
  DisplayText {
   widgetClassPath=&CommonTextBlock
   bIsVariable=true
   parent=VerticalBox_0
   widget=DisplayText
   properties {
    scrollStyle=&TextScrollStyle-Base_C
    text=Parameter Name
    font {
     bForceMonospaced=false
     bMaterialIsStencil=false
     fontMaterial=null
     fontObject=&BurbankSmall
     letterSpacing=0
     monospacedWidth=1
     size=16
     skewAmount=0
     typefaceFontName=Bold
     outlineSettings {
      bApplyOutlineToDropShadows=false
      bMiteredCorners=false
      bSeparateFillAlpha=false
      outlineMaterial=null
      outlineSize=0
      outlineColor=<rgba 0 0 0 1>
     }
    }
    shadowOffset=<0 0>
   }
   slot {class=VerticalBoxSlot, properties {horizontalAlignment=HAlign_Center, verticalAlignment=VAlign_Top, padding=<ltrb 0 0 0 10>}}
  }
  HorizontalBox_Selector {
   widgetClassPath=&HorizontalBox
   bIsVariable=true
   parent=VerticalBox_0
   widget=HorizontalBox_Selector
   properties {flowDirectionPreference=LeftToRight}
   slot {class=VerticalBoxSlot, properties {size {sizeRule=Fill, value=1}}}
  }
  ButtonLeft {
   widgetClassPath=&IconTextButton
   bIsVariable=true
   parent=HorizontalBox_Selector
   widget=ButtonLeft
   properties {
    bIsFocusable=false
    buttonText=""
    clickMethod=MouseDown
    contentAlignment=HAlign_Center
    onButtonBaseClicked=(&self:WidgetTree.ButtonLeft.HandleOnButtonClicked)
    onButtonBaseDoubleClicked=(&self:WidgetTree.ButtonLeft.HandleOnButtonDoubleClicked)
    onButtonBaseHovered=(&self:WidgetTree.ButtonLeft.HandleOnButtonHovered)
    onButtonBaseUnhovered=(&self:WidgetTree.ButtonLeft.HandleOnButtonUnhovered)
    onSelectedChangedBase=(&self:WidgetTree.ButtonLeft.HandleOnSelectedChanged)
    style=/FortUILibrary/Foundation/Buttons/ButtonStyles/Athena/Athena_ButtonStyle_BlueMenuButton_Settings.Athena_ButtonStyle_BlueMenuButton_Settings_C
    iconBrush {
     bIsDynamicallyLoaded=false
     drawAs=Image
     imageType=NoImage
     mirroring=NoMirror
     resourceName=None
     resourceObject=/FortUILibrary/Foundation/Buttons/Materials/M-Button-Outline-Base-Left
     tiling=NoTile
     imageSize=<42 42>
     margin=<ltrb 0 0 0 0>
     outlineSettings {
      bUseBrushTransparency=false
      roundingType=HalfHeightRadius
      width=0
      color {colorUseRule=UseColor_Specified, specifiedColor=<rgba 0 0 0 0>}
      cornerRadii=<0 0 0 1>
     }
     tintColor {colorUseRule=UseColor_Specified, specifiedColor=<rgba 1 1 1 1>}
     uVRegion {bIsValid=false, max=<0 0>, min=<0 0>}
    }
   }
   slot {class=HorizontalBoxSlot, properties {horizontalAlignment=HAlign_Left}}
  }
  Overlay_0 {
   widgetClassPath=&Overlay
   parent=HorizontalBox_Selector
   widget=Overlay_0
   slot {class=HorizontalBoxSlot, properties {verticalAlignment=VAlign_Center}}
  }
  Image_103 {
   widgetClassPath=&Image
   bIsVariable=true
   parent=Overlay_0
   widget=Image_103
   properties {brush {
     bIsDynamicallyLoaded=false
     drawAs=Image
     imageType=NoImage
     mirroring=NoMirror
     resourceName=None
     resourceObject=null
     tiling=NoTile
     imageSize=<32 32>
     margin=<ltrb 0 0 0 0>
     outlineSettings {
      bUseBrushTransparency=false
      roundingType=HalfHeightRadius
      width=0
      color {colorUseRule=UseColor_Specified, specifiedColor=<rgba 0 0 0 0>}
      cornerRadii=<0 0 0 0>
     }
     tintColor {colorUseRule=UseColor_Specified, specifiedColor=<rgba 0 0 0.1406 1>}
     uVRegion {bIsValid=false, max=<0 0>, min=<0 0>}
    }}
   slot {class=OverlaySlot, properties {horizontalAlignment=HAlign_Fill, verticalAlignment=VAlign_Fill}}
  }
  SettingsRotator {
   widgetClassPath=/FortniteCoreUI/Settings/SettingsRotator
   bIsVariable=true
   parent=Overlay_0
   widget=SettingsRotator
   properties {renderTransform {angle=0, scale=<1.3 1.3>, shear=<0 0>, translation=<0 0>}}
   slot {class=OverlaySlot, properties {horizontalAlignment=HAlign_Fill, padding=<ltrb 30 20 30 20>}}
  }
  ButtonRight {
   widgetClassPath=&IconTextButton
   bIsVariable=true
   parent=HorizontalBox_Selector
   widget=ButtonRight
   properties {
    bIsFocusable=false
    buttonText=""
    clickMethod=MouseDown
    contentAlignment=HAlign_Center
    onButtonBaseClicked=(&self:WidgetTree.ButtonRight.HandleOnButtonClicked)
    onButtonBaseDoubleClicked=(&self:WidgetTree.ButtonRight.HandleOnButtonDoubleClicked)
    onButtonBaseHovered=(&self:WidgetTree.ButtonRight.HandleOnButtonHovered)
    onButtonBaseUnhovered=(&self:WidgetTree.ButtonRight.HandleOnButtonUnhovered)
    onSelectedChangedBase=(&self:WidgetTree.ButtonRight.HandleOnSelectedChanged)
    style=&Athena_ButtonStyle_BlueMenuButton_Settings_C
    iconBrush {
     bIsDynamicallyLoaded=false
     drawAs=Image
     imageType=NoImage
     mirroring=NoMirror
     resourceName=None
     resourceObject=&M-Button-Outline-Base-Left
     tiling=NoTile
     imageSize=<42 42>
     margin=<ltrb 0 0 0 0>
     outlineSettings {
      bUseBrushTransparency=false
      roundingType=HalfHeightRadius
      width=0
      color {colorUseRule=UseColor_Specified, specifiedColor=<rgba 0 0 0 0>}
      cornerRadii=<0 0 0 1>
     }
     tintColor {colorUseRule=UseColor_Specified, specifiedColor=<rgba 1 1 1 1>}
     uVRegion {bIsValid=false, max=<0 0>, min=<0 0>}
    }
    renderTransform {angle=0, scale=<-1 1>, shear=<0 0>, translation=<0 0>}
   }
   slot {class=HorizontalBoxSlot, properties {horizontalAlignment=HAlign_Right}}
  }
 ]
}
