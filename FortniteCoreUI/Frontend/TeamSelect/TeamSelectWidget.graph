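# Creative team select modal, based on FortCreativeTeamSelectWidget, that lets a player change team.
# On construct it reads the playlist teams from the Athena game state, fills the rotator with numbered labels and preselects the player's team.
# Activation updates the current team text.
# Apply, labeled Respawn, sends the chosen team through ServerSetModeratorTeam in moderator mode or ServerSetTeam otherwise, then pops the panel, as do Cancel and touch close.
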WidgetBlueprint /FortniteCoreUI/Frontend/TeamSelect/TeamSelectWidget
graphs [
 EventGraph {nodes [
   n1 = "On Button Clicked (Button_CloseTouch)"!() -> (OutputDelegate, then, Button:Common Button Legacy Object Reference) {is=title K2Node_ComponentBoundEvent}
   PopActivatablePanelInModalLayer!(
    execute=[@n6.then, @n7.then, @n4.Invalid, @n1.then, @n8.then, @n3.Invalid]
    Panel:Common Activatable Panel Legacy Object Reference=Self-Reference()
   ) -> (then)
   n2 = "Selection Changed (TeamRotator)"!() -> (OutputDelegate, then, Selected Index:int=0) {is=title K2Node_ComponentBoundEvent}
   > Set Team Index(Team Index:int=@n2.Selected Index) -> (Output_Get:int=0)
   "On Button Clicked (ApplyButton)"!() -> (OutputDelegate, then, Button:Common Button Legacy Object Reference) {is=title K2Node_ComponentBoundEvent}
   > n3 = Get Athena Player Controller!(Exec:Exec) -> (Valid:Exec, Invalid:Exec, Controller:Fort Player Controller Athena Object Reference) {is=title K2Node_MacroInstance}
   > n4 = Get Validated Array Item!(Exec:Exec, Array:Array of Bytes=Get Available Teams() -> (Available Teams:Array of Bytes), Index:int=Get Team Index() -> (Team Index:int=0)) -> (Valid:Exec, Invalid:Exec, Item:Byte) {is=title K2Node_MacroInstance}
   > n5 = Branch(Condition=IsCreativeModeratorModeActive!(self:Fort Player Controller Gameplay Object Reference=@n3) -> (ReturnValue:bool))
   > n6 = ServerSetModeratorTeam!(execute, self:Fort Player Controller Athena Object Reference=@n3, InTeam:Byte=@n4) -> (then)
   n7 = ServerSetTeam!(execute=@n5.else, self:Fort Player Controller Athena Object Reference=@n3, InTeam:Byte=@n4) -> (then)
   n8 = "On Button Clicked (CancelButton)"!() -> (OutputDelegate, then, Button:Common Button Legacy Object Reference) {is=title K2Node_ComponentBoundEvent}
   Event On Activated()
   > n9 = Get Athena Player Controller()
   n10 = GetActorTeam!(Actor:Actor Object Reference=@n9) -> (ReturnValue:Byte=0)
   > "SetText (Text)"(
    self=Get Text_CurrentTeam() -> (Text_CurrentTeam:Common Text Object Reference)
    InText=Format Text!(
     Format="LOCTEXT(\"Current Team: {Team}\")"
     Team:Text=Get Team String!(Team Num:Byte=@n10) -> (Team Text:Text) {is=title K2Node_MacroInstance}
    ) -> (Result)
   )
   Event Construct()
   > n11 = Get Athena Game State!(Exec:Exec) -> (Valid:Exec, Invalid:Exec, Game State:Fort Game State Athena Object Reference) {is=title K2Node_MacroInstance}
   n12 = PrintString!(
    execute=@n11.Invalid
    InString=TeamSelectWidget: Failed to get game state
    bPrintToScreen
    bPrintToLog
    TextColor=(R=0.66,G=0,B=0,A=1)
    Duration=5
    Key
   ) -> (then)
   Initialize!(
    execute=[@n23.Completed, @n24.then, @n12.then]
    self:Rotator Selector Teams 1 Object Reference=Get TeamRotator() -> (TeamRotator:Rotator Selector Teams 1 Object Reference)
    Display Text:Text
    Options:Array of Texts=Get Available Teams Text() -> (Available Teams Text:Array of Texts)
    Hover Text:Text=GetEmptyText!() -> (ReturnValue:Text)
    Tab Tooltip:Common Text Object Reference=Get TooltipDisplay() -> (TooltipDisplay:Common Text Object Reference)
   ) -> (then)
   > n13 = Get Athena Player Controller()
   n14 = Find Item!(TargetArray:Array of Bytes=Get Available Teams(), ItemToFind:"Byte (by ref)"=GetActorTeam(Actor=@n13) -> (ReturnValue=0)) -> (ReturnValue)
   > n15 = Branch(Condition="Not Equal (Integer)"!(A:int=@n14, B:int=-1) -> (ReturnValue:bool) {is=title K2Node_CallFunction})
   n16 = Get TeamRotator()
   > n17 = Update Row Selector!(execute, self:Rotator Selector Teams 1 Object Reference=@n16, Currently Selected:int=@n14) -> (then)
   n18 = Get TeamRotator()
   Set Team Index(execute=[@n19.then, @n17.then], Team Index=Get SelectedIndex!(self:Rotator Selector Teams 1 Object Reference=@n18) -> (SelectedIndex:int=0)) -> (Output_Get=0)
   > Center On Widget(self:Rotator Selector Teams 1 Object Reference=@n18)
   n19 = Update Row Selector(execute=[@n15.else, @n13.Invalid], self=@n16, Currently Selected=0)
   n20 = GetPlaylistTeams!(self:Fort Game State Athena Object Reference=@n11) -> (OutTeams:Array of Bytes, ReturnValue:bool)
   n21 = Branch(execute=@n11.Valid, Condition=@n20.ReturnValue)
   > n22 = Set Available Teams(Available Teams:Array of Bytes=@n20) -> (Output_Get:Array of Bytes)
   > Resize!(execute, TargetArray:Array of Texts=Get Available Teams Text(), Size:int=Length(TargetArray:Array of Bytes=@n22)) -> (then) {is=title K2Node_CallArrayFunction}
   > n23 = For Each Loop(Array:Array of Bytes=@n22) -> (Array Element:Byte)
   > Set Array Elem!(
    execute
    TargetArray:Array of Texts=Get Available Teams Text()
    Index:int=@n23.Array Index
    Item:"Text (by ref)"=Format Text!(
     Format="LOCTEXT(\"Team {TeamNumber}\")"
     TeamNumber:int=int + int(A=@n23.Array Index, B=1) -> (ReturnValue=0) {is=title K2Node_CommutativeAssociativeBinaryOperator}
    ) -> (Result)
    bSizeToFit:bool=false
   ) -> (then)
   n24 = PrintString(execute=@n21.else, InString=TeamSelectWidget: Failed to get playlist teams, TextColor=(R=0.66,G=0,B=0,A=1), Duration=5)
  ]}
 Get Athena Game State {nodes [
   Inputs!() -> (Exec:Exec)
   > n1 = Cast To FortGameStateAthena!(execute, Object=GetGameState!() -> (ReturnValue:Game State Base Object Reference)) -> (then, CastFailed, AsFort Game State Athena:Fort Game State Athena Object Reference) {is=title K2Node_DynamicCast}
   > n2 = Is Valid(InputObject=@n1)
   > Outputs!(Valid:Exec, Invalid:Exec=[@n1.CastFailed, @n2.Is Not Valid], Game State:Fort Game State Athena Object Reference=@n1)
  ]}
 Get Athena Player Controller {nodes [
   Inputs()
   > n1 = Cast To FortPlayerControllerAthena(Object=GetOwningPlayer())
   > n2 = Is Valid(InputObject=@n1)
   > Outputs!(Valid, Invalid=[@n1.CastFailed, @n2.Is Not Valid], Controller:Fort Player Controller Athena Object Reference=@n1)
  ]}
 Get Team String {nodes [
   n1 = Find Item(TargetArray=Get Available Teams(), ItemToFind=Inputs!() -> (Team Num:Byte))
   n2 = Get Available Teams Text()
   Outputs!(Team Text:Text=Select(
     Option 0:Text=GetEmptyText()
     Option 1:Text="Get (a ref)"(Array:Array of Texts=@n2, Dimension 1=@n1) -> (Output:"Text (by ref)")
     Index=Is Valid Index!(TargetArray:Array of Texts=@n2, IndexToTest:int=@n1) -> (ReturnValue:bool)
    ) -> (ReturnValue:Text))
  ]}
 Get Validated Array Item {nodes [
   n1 = Inputs!() -> (Exec:Exec, Array:Array of Wildcards, Index:int)
   > n2 = Branch(Condition=Is Valid Index(TargetArray:Array of Wildcards=@n1, IndexToTest=@n1.Index))
   > Outputs!(Valid:Exec, Invalid:Exec=@n2.else, Item:Wildcard="Get (a copy)"(Array:Array of Wildcards=@n1, Dimension 1=@n1.Index) -> (Output:Wildcard))
  ]}
]
variables [
 Available Teams {type="array<integer>"}
 Available Teams Text {type="array<string>", default []}
 Team Index {type=integer}
]
widgets {
 info {
  inheritedWidgetCount=0
  namedSlotCount=1
  parentClass=/Script/FortniteUI.FortCreativeTeamSelectWidget
  rootWidgetClass=&Border
  widgetCount=26
 }
 widgets [
  RootBorder {
   widgetClassPath=&Border
   widget=RootBorder
   properties {
    background {
     bIsDynamicallyLoaded=false
     drawAs=NoDrawType
     imageType=NoImage
     mirroring=NoMirror
     resourceName=None
     resourceObject=null
     tiling=NoTile
     imageSize=<32 32>
     margin=<ltrb 0 0 0 0>
     outlineSettings {
      bUseBrushTransparency=false
      roundingType=HalfHeightRadius
      width=0
      color {colorUseRule=UseColor_Specified, specifiedColor=<rgba 0 0 0 0>}
      cornerRadii=<0 0 0 0>
     }
     tintColor {colorUseRule=UseColor_Specified, specifiedColor=<rgba 1 1 1 1>}
     uVRegion {bIsValid=false, max=<0 0>, min=<0 0>}
    }
    padding=<ltrb 0 0 0 0>
   }
  }
  BackgroundBlur_2 {
   widgetClassPath=&BackgroundBlur
   parent=RootBorder
   widget=BackgroundBlur_2
   properties {blurStrength=15}
   slot {class=BorderSlot, properties {padding=<ltrb 0 0 0 0>}}
  }
  Lightbox {
   widgetClassPath=/Game/UI/Foundation/Lightbox/Lightbox
   bIsVariable=true
   parent=BackgroundBlur_2
   widget=Lightbox
   slot {class=BackgroundBlurSlot, properties {padding=<ltrb 0 0 0 0>}}
  }
  RootOverlay {widgetClassPath=&Overlay, namedSlotHost=Lightbox, widget=RootOverlay}
  TouchToCloseZone {
   widgetClassPath=&CommonBorder
   parent=RootOverlay
   widget=TouchToCloseZone
   properties {horizontalAlignment=HAlign_Center, verticalAlignment=VAlign_Center, visibility=Visible, padding=<ltrb 0 0 0 0>}
   slot {class=OverlaySlot, properties {horizontalAlignment=HAlign_Fill, verticalAlignment=VAlign_Fill}}
  }
  SizeBox_4 {
   widgetClassPath=&SizeBox
   parent=TouchToCloseZone
   widget=SizeBox_4
   properties {bOverride_WidthOverride=true, widthOverride=1000}
   slot {class=BorderSlot, properties {horizontalAlignment=HAlign_Center, verticalAlignment=VAlign_Center, padding=<ltrb 0 0 0 0>}}
  }
  Overlay_3 {widgetClassPath=&Overlay, parent=SizeBox_4, widget=Overlay_3, slot {class=SizeBoxSlot}}
  Image_149 {
   widgetClassPath=&Image
   parent=Overlay_3
   widget=Image_149
   properties {colorAndOpacity=<rgba 0 0.0222 0.2874 1>}
   slot {class=OverlaySlot, properties {horizontalAlignment=HAlign_Fill, verticalAlignment=VAlign_Fill}}
  }
  Image_515 {
   widgetClassPath=&Image
   parent=Overlay_3
   widget=Image_515
   properties {
    brush {
     bIsDynamicallyLoaded=false
     drawAs=Image
     imageType=NoImage
     mirroring=NoMirror
     resourceName=None
     resourceObject=/Game/Athena/UI/InGame/Creative/Art/M_UI_BoltGradientBorder_Creative_IslandSelectFooter
     tiling=NoTile
     imageSize=<32 64>
     margin=<ltrb 0 0 0 0>
     outlineSettings {
      bUseBrushTransparency=false
      roundingType=HalfHeightRadius
      width=0
      color {colorUseRule=UseColor_Specified, specifiedColor=<rgba 0 0 0 0>}
      cornerRadii=<0 0 0 0>
     }
     tintColor {colorUseRule=UseColor_Specified, specifiedColor=<rgba 1 1 1 1>}
     uVRegion {bIsValid=false, max=<0 0>, min=<0 0>}
    }
    colorAndOpacity=<rgba 0.0052 0.0578 0.3916 1>
   }
   slot {class=OverlaySlot, properties {horizontalAlignment=HAlign_Fill, verticalAlignment=VAlign_Bottom}}
  }
  VerticalBox_3 {
   widgetClassPath=&VerticalBox
   parent=Overlay_3
   widget=VerticalBox_3
   properties {navigation=&self:WidgetTree.VerticalBox_3.WidgetNavigation_0}
   slot {class=OverlaySlot, properties {horizontalAlignment=HAlign_Fill}}
  }
  CommonBorder_Modal {widgetClassPath=&Overlay, parent=VerticalBox_3, widget=CommonBorder_Modal, slot {class=VerticalBoxSlot}}
  Image_180 {
   widgetClassPath=&Image
   parent=CommonBorder_Modal
   widget=Image_180
   properties {
    brush {
     bIsDynamicallyLoaded=false
     drawAs=Image
     imageType=NoImage
     mirroring=NoMirror
     resourceName=None
     resourceObject=/Game/Athena/UI/InGame/Creative/Art/M_UI_BoltGradientBorder_Creative_IslandSelectHeader
     tiling=NoTile
     imageSize=<32 73.4941>
     margin=<ltrb 0 0 0 0>
     outlineSettings {
      bUseBrushTransparency=false
      roundingType=HalfHeightRadius
      width=0
      color {colorUseRule=UseColor_Specified, specifiedColor=<rgba 0 0 0 0>}
      cornerRadii=<0 0 0 0>
     }
     tintColor {colorUseRule=UseColor_Specified, specifiedColor=<rgba 1 1 1 1>}
     uVRegion {bIsValid=false, max=<0 0>, min=<0 0>}
    }
    renderTransform {angle=0, scale=<1 -1>, shear=<0 0>, translation=<0 -8>}
   }
   slot {class=OverlaySlot, properties {horizontalAlignment=HAlign_Fill, verticalAlignment=VAlign_Fill}}
  }
  HorizontalBox_1 {
   widgetClassPath=&HorizontalBox
   parent=CommonBorder_Modal
   widget=HorizontalBox_1
   slot {class=OverlaySlot, properties {horizontalAlignment=HAlign_Fill, padding=<ltrb 0 10 0 30>}}
  }
  Spacer_0 {widgetClassPath=&Spacer, parent=HorizontalBox_1, widget=Spacer_0, properties {size=<30 100>}, slot {class=HorizontalBoxSlot}}
  Image_1 {
   widgetClassPath=&Image
   bIsVariable=true
   parent=HorizontalBox_1
   widget=Image_1
   properties {brush {
     bIsDynamicallyLoaded=false
     drawAs=Image
     imageType=NoImage
     mirroring=NoMirror
     resourceName=None
     resourceObject=/FortniteCoreUI/Athena/HUD/Art/AliveTeamsIco
     tiling=NoTile
     imageSize=<38 38>
     margin=<ltrb 0 0 0 0>
     outlineSettings {
      bUseBrushTransparency=false
      roundingType=HalfHeightRadius
      width=0
      color {colorUseRule=UseColor_Specified, specifiedColor=<rgba 0 0 0 0>}
      cornerRadii=<0 0 0 0>
     }
     tintColor {colorUseRule=UseColor_Specified, specifiedColor=<rgba 1 1 1 1>}
     uVRegion {bIsValid=false, max=<0 0>, min=<0 0>}
    }}
   slot {class=HorizontalBoxSlot, properties {verticalAlignment=VAlign_Center}}
  }
  Title {
   widgetClassPath=&CommonTextBlock
   bIsVariable=true
   parent=HorizontalBox_1
   widget=Title
   properties {
    text=TEAM SELECT
    colorAndOpacity {colorUseRule=UseColor_Specified, specifiedColor=<rgba 1 1 1 0.8>}
    font {
     bForceMonospaced=false
     bMaterialIsStencil=false
     fontMaterial=null
     fontObject=&Burbank
     letterSpacing=0
     monospacedWidth=1
     size=48
     skewAmount=0
     typefaceFontName=Burbank-Black
     outlineSettings {
      bApplyOutlineToDropShadows=false
      bMiteredCorners=false
      bSeparateFillAlpha=false
      outlineMaterial=null
      outlineSize=0
      outlineColor=<rgba 0 0 0 1>
     }
    }
    shadowOffset=<0 0>
   }
   slot {class=HorizontalBoxSlot, properties {horizontalAlignment=HAlign_Left, verticalAlignment=VAlign_Center, padding=<ltrb 8 0 0 0>}}
  }
  VisibilityBorder_Touch {
   widgetClassPath=&CommonHardwareVisibilityBorder
   parent=HorizontalBox_1
   widget=VisibilityBorder_Touch
   properties {visibilityQuery {
     autoDescription=" ANY( Input.Touch )"
     tokenStreamVersion=0
     userDescription=""
     queryTokenStream [0,1,1,1,0]
     tagDictionary [{tagName=Input.Touch}]
    }}
   slot {
    class=HorizontalBoxSlot
    properties {horizontalAlignment=HAlign_Right, verticalAlignment=VAlign_Top, padding=<ltrb 0 0 10 0>, size {sizeRule=Fill, value=1}}
   }
  }
  Button_CloseTouch {
   widgetClassPath=/FortUILibrary/Foundation/Buttons/CloseButton
   bIsVariable=true
   parent=VisibilityBorder_Touch
   widget=Button_CloseTouch
   properties {
    onButtonBaseClicked=(&self:WidgetTree.Button_CloseTouch.HandleOnButtonClicked)
    onButtonBaseDoubleClicked=(&self:WidgetTree.Button_CloseTouch.HandleOnButtonDoubleClicked)
    onButtonBaseHovered=(&self:WidgetTree.Button_CloseTouch.HandleOnButtonHovered)
    onButtonBaseUnhovered=(&self:WidgetTree.Button_CloseTouch.HandleOnButtonUnhovered)
    onSelectedChangedBase=(&self:WidgetTree.Button_CloseTouch.HandleOnSelectedChanged)
    renderTransform {angle=0, scale=<0.8 0.8>, shear=<0 0>, translation=<0 0>}
   }
   slot {class=BorderSlot, properties {padding=<ltrb 0 0 0 0>}}
  }
  TooltipDisplay {
   widgetClassPath=&CommonTextBlock
   bIsVariable=true
   parent=VerticalBox_3
   widget=TooltipDisplay
   properties {
    autoWrapText=true
    style=/FortUILibrary/Styles/Typography/Foundation/TextStyle-Base-XS-Blue.TextStyle-Base-XS-Blue_C
    visibility=Collapsed
    colorAndOpacity {colorUseRule=UseColor_Specified, specifiedColor=<rgba 0.008 0.011 0.0262 1>}
    font {
     bForceMonospaced=false
     bMaterialIsStencil=false
     fontMaterial=null
     fontObject=&NotoSans
     letterSpacing=0
     monospacedWidth=1
     size=16
     skewAmount=0
     typefaceFontName=NotoSans-Regular
     outlineSettings {
      bApplyOutlineToDropShadows=false
      bMiteredCorners=false
      bSeparateFillAlpha=false
      outlineMaterial=null
      outlineSize=0
      outlineColor=<rgba 0 0 0 1>
     }
    }
    shadowOffset=<0 0>
   }
   slot {class=VerticalBoxSlot, properties {verticalAlignment=VAlign_Top, padding=<ltrb 32 60 32 0>}}
  }
  HorizontalBox_0 {
   widgetClassPath=&HorizontalBox
   parent=VerticalBox_3
   widget=HorizontalBox_0
   slot {class=VerticalBoxSlot, properties {padding=<ltrb 20 40 20 30>}}
  }
  Text_CurrentTeam {
   widgetClassPath=&CommonTextBlock
   bIsVariable=true
   parent=HorizontalBox_0
   widget=Text_CurrentTeam
   properties {
    mobileFontSizeMultiplier=1.2
    text=Current Team
    font {
     bForceMonospaced=false
     bMaterialIsStencil=false
     fontMaterial=null
     fontObject=&BurbankSmall
     letterSpacing=0
     monospacedWidth=1
     size=24
     skewAmount=0
     typefaceFontName=Bold
     outlineSettings {
      bApplyOutlineToDropShadows=true
      bMiteredCorners=false
      bSeparateFillAlpha=true
      outlineMaterial=&M-FontShadow
      outlineSize=3
      outlineColor=<rgba 0 0 0 0.5>
     }
    }
    shadowOffset=<0 0>
   }
   slot {
    class=HorizontalBoxSlot
    properties {horizontalAlignment=HAlign_Center, verticalAlignment=VAlign_Center, padding=<ltrb 0 17 0 0>, size {sizeRule=Fill, value=1}}
   }
  }
  TeamRotator {
   widgetClassPath=&/RotatorSelectorTeams1
   bIsVariable=true
   parent=HorizontalBox_0
   widget=TeamRotator
   slot {class=HorizontalBoxSlot, properties {horizontalAlignment=HAlign_Center, verticalAlignment=VAlign_Center, size {sizeRule=Fill, value=1.2}}}
  }
  HorizontalBox_6 {
   widgetClassPath=&HorizontalBox
   parent=VerticalBox_3
   widget=HorizontalBox_6
   properties {navigation=&self:WidgetTree.HorizontalBox_6.WidgetNavigation_0}
   slot {class=VerticalBoxSlot, properties {horizontalAlignment=HAlign_Center, padding=<ltrb 10 50 20 40>}}
  }
  ApplyButton {
   widgetClassPath=&IconTextButton
   bIsVariable=true
   parent=HorizontalBox_6
   widget=ApplyButton
   properties {
    buttonText=Respawn
    minWidth=312
    navigation=&self:WidgetTree.ApplyButton.WidgetNavigation_1
    onButtonBaseClicked=(&self:WidgetTree.ApplyButton.HandleOnButtonClicked)
    onButtonBaseDoubleClicked=(&self:WidgetTree.ApplyButton.HandleOnButtonDoubleClicked)
    onButtonBaseHovered=(&self:WidgetTree.ApplyButton.HandleOnButtonHovered)
    onButtonBaseUnhovered=(&self:WidgetTree.ApplyButton.HandleOnButtonUnhovered)
    onSelectedChangedBase=(&self:WidgetTree.ApplyButton.HandleOnSelectedChanged)
    style=/FortUILibrary/Foundation/Buttons/ButtonStyles/ButtonStyle_Feature_M_Yellow.ButtonStyle_Feature_M_Yellow_C
    triggeringInputAction {dataTable=&InputActionDataTable, rowName=Apply}
   }
   slot {class=HorizontalBoxSlot}
  }
  Spacer_99 {widgetClassPath=&Spacer, parent=HorizontalBox_6, widget=Spacer_99, properties {size=<50 1>}, slot {class=HorizontalBoxSlot}}
  CancelButton {
   widgetClassPath=&IconTextButton
   bIsVariable=true
   parent=HorizontalBox_6
   widget=CancelButton
   properties {
    buttonText=Cancel
    minWidth=312
    navigation=&self:WidgetTree.CancelButton.WidgetNavigation_1
    onButtonBaseClicked=(&self:WidgetTree.CancelButton.HandleOnButtonClicked)
    onButtonBaseDoubleClicked=(&self:WidgetTree.CancelButton.HandleOnButtonDoubleClicked)
    onButtonBaseHovered=(&self:WidgetTree.CancelButton.HandleOnButtonHovered)
    onButtonBaseUnhovered=(&self:WidgetTree.CancelButton.HandleOnButtonUnhovered)
    onSelectedChangedBase=(&self:WidgetTree.CancelButton.HandleOnSelectedChanged)
    style=/FortUILibrary/Foundation/Buttons/ButtonStyles/Athena/Athena_ButtonStyle_AngledBlueMenuCreativeButton.Athena_ButtonStyle_AngledBlueMenuCreativeButton_C
    triggeringInputAction {dataTable=&InputActionDataTable, rowName=Cancel}
   }
   slot {class=HorizontalBoxSlot}
  }
 ]
}
